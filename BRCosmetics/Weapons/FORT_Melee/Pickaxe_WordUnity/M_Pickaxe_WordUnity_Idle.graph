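# Translucent unlit material for Niagara mesh particles in the WordUnity pickaxe's idle effect.
# A Color parameter feeds MF_FX_EmissiveAdjustments for emissive.
# For opacity, a narrow SmoothStep on an offset V gradient becomes a hard line via Frac and Ceil, with a faint fill on one side.
# A ceiled, inverted DepthFade adds hard intersection highlights within that shape before the Alpha parameter scales the result.
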Material /BRCosmetics/Weapons/FORT_Melee/Pickaxe_WordUnity/M_Pickaxe_WordUnity_Idle
graph {
 properties {
  bAutomaticallySetUsageInEditor=false
  bUsedWithNiagaraMeshParticles=true
  blendMode=BLEND_Translucent
  shadingModel=MSM_Unlit
  usedShadingModels=Unlit
 }
 nodes [
  MaterialFunctionCall(Emissive=VectorParameter() {parameterName=Color, defaultValue=<rgba 0.66 0.033 0.5555 1>}.RGB) {materialFunction=&MF_FX_EmissiveAdjustments}
  n1 = TextureCoordinate()
  n2 = Constant2Vector() {g=0.2}
  n3 = SmoothStep(Value=MaterialFunctionCall(UV Channel=Add(A=@n1, B=@n2)) {materialFunction=&LinearGradient}.VGradient) {constMax=0.5, constMin=0.47}
  n4 = Max(A=Ceil(None=Frac(None=@n3)), B=Multiply(A=@n3) {constB=0.15})
  n5 = Ceil(None=OneMinus(None=DepthFade() {fadeDistanceDefault=0.6}))
  Multiply(A=Multiply(A=Ceil(None=@n4), B=Add(A=@n4, B=@n5)), B=ScalarParameter() {defaultValue=1.5, parameterName=Alpha})
 ]
}
