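# Translucent unlit Niagara trail material for the WordUnity pickaxe, drawing a two-tone stylized streak.
# A UV-shaped trail mask, modulated by two opposing panned light-ray samples, has two multiplied Voronoi noises subtracted to break it up.
# Step thresholds split the result into inner and outer colors tinted by particle color, plus matching alpha levels.
# Opacity adds depth fade off mobile, particle alpha and a dynamic-parameter sphere-mask cutout.
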Material /BRCosmetics/Weapons/FORT_Melee/Pickaxe_WordUnity/M_Pickaxe_WordUnity_Trail
graph {
 properties {
  bUsedWithNiagaraMeshParticles=true
  bUsedWithNiagaraRibbons=true
  bUsedWithNiagaraSprites=true
  blendMode=BLEND_Translucent
  pixelDepthOffsetMode=PDOM_Legacy
  shadingModel=MSM_Unlit
  usedShadingModels=Unlit
 }
 nodes [
  n1 = TextureCoordinate()
  n2 = ComponentMask(None=@n1) {g=true}
  n3 = LinearInterpolate(A=Sine(None=Multiply(A=@n2) {constB=0.5}), Alpha=@n2) {constB=0}
  n4 = OneMinus(None=ComponentMask(None=@n1) {r=true})
  n5 = LinearInterpolate(
   A=@n4
   B=Power(Base=@n4, Exp=ScalarParameter() {defaultValue=3, group=Trail Alpha Values, parameterName=Trail_Shape_Intensity}) {constExponent=3}
   Alpha=@n2
  )
  n6 = Multiply(A=Saturate(None=Multiply(A=@n3, B=@n5)), B=ScalarParameter() {defaultValue=3, group=Trail Alpha Values, parameterName=Trail_Shape_Size}) {constB=3}
  n7 = MaterialFunctionCall(Float2=TextureCoordinate()) {materialFunction=&BreakOutFloat2Components}
  n8 = Multiply(A=@n7.G)
  n9 = TextureSampleParameter2D(UVs=Panner(Coordinate=AppendVector(A=@n8, B=@n7.R)) {speedX=0.3}) {
   group=Motion Lines
   parameterName=Motion Lines Texture
   samplerSource=SSM_Wrap_WorldGroupSettings
   texture=/Game/Effects/Fort_Effects/Textures/Outlander/T_InTheZone_LightRays_01
  }
  n10 = Multiply(A=@n7.G) {constB=2}
  n11 = TextureSampleParameter2D(UVs=Panner(Coordinate=AppendVector(A=@n10, B=@n7.R)) {speedX=-0.7}) {
   group=Motion Lines
   parameterName=Motion Lines Texture
   samplerSource=SSM_Wrap_WorldGroupSettings
   texture=&T_InTheZone_LightRays_01
  }
  n12 = Multiply(A=Add(A=Power(Base=@n9.R) {constExponent=1}, B=Power(Base=@n11.R))) {constB=0.2}
  n13 = ScalarParameter() {defaultValue=0.3, group=Voronoi Noise 1, parameterName=Voronoi_1_Offset_U, sortPriority=50}
  n14 = MaterialFunctionCall(Float2=TextureCoordinate()) {materialFunction=&BreakOutFloat2Components}
  n15 = Multiply(A=ScalarParameter() {defaultValue=0.2, group=Voronoi Noise 1, parameterName=Voronoi_1_Scale_U}, B=@n14.R) {constB=0.5}
  n16 = ScalarParameter() {defaultValue=1, group=Voronoi Noise 1, parameterName=Voronoi_1_Scale_V}
  n17 = TextureSampleParameter2D(UVs=AppendVector(A=Add(A=@n13, B=@n15) {constB=0.3}, B=Multiply(A=@n14.G, B=@n16) {constB=3})) {
   group=Voronoi Noise 1
   parameterName=TexVoronoi1
   samplerType=SAMPLERTYPE_LinearGrayscale
   texture=/BRCosmetics/Weapons/FORT_Melee/Pickaxe_FearlessFlightMenace/FX/Textures/T_FearlessFlight_Menace_Voronoi_1
  }
  n18 = Multiply(
   A=Power(Base=@n17.RGB, Exp=ScalarParameter() {defaultValue=2, group=Voronoi Noise 1, parameterName=Voronoi_1_Power})
   B=ScalarParameter() {defaultValue=2, group=Voronoi Noise 1, parameterName=Voronoi_1_Hardness}
  ) {constB=2}
  n19 = Multiply(A=@n14.R, B=ScalarParameter() {defaultValue=0.2, group=Voronoi Noise 2, parameterName=Voronoi_2_Scale_U}) {constB=0.5}
  n20 = ScalarParameter() {defaultValue=0.3, group=Voronoi Noise 2, parameterName=Voronoi_2_Offset_U, sortPriority=50}
  n21 = ScalarParameter() {defaultValue=1, group=Voronoi Noise 2, parameterName=Voronoi_2_Scale_V}
  n22 = TextureSampleParameter2D(UVs=AppendVector(A=Add(A=@n19, B=@n20) {constB=0.3}, B=Multiply(A=@n14.G, B=@n21))) {
   group=Voronoi Noise 2
   parameterName=TexVoronoi2
   samplerType=SAMPLERTYPE_LinearGrayscale
   texture=/BRCosmetics/Weapons/FORT_Melee/Pickaxe_FearlessFlightMenace/FX/Textures/T_FearlessFlight_Menace_Voronoi_2
  }
  n23 = Multiply(
   A=Power(Base=@n22.RGB, Exp=ScalarParameter() {defaultValue=0.8, group=Voronoi Noise 2, parameterName=Voronoi_2_Power}) {constExponent=0.8}
   B=ScalarParameter() {defaultValue=1, group=Voronoi Noise 2, parameterName=Voronoi_2_Hardness}
  )
  n24 = Subtract(
   A=Multiply(A=Saturate(None=@n6), B=LinearInterpolate(Alpha=@n12) {constA=0.85})
   B=Multiply(A=Saturate(None=@n18), B=Saturate(None=@n23)) {constA=1}
  )
  n25 = Step(Y=OneMinus(None=ScalarParameter() {defaultValue=1, group=Trail Color, parameterName=Outer_Color_Coverage, sliderMax=1}), X=@n24) {constX=0.9}
  n26 = Step(Y=OneMinus(None=ScalarParameter() {defaultValue=0.48, group=Trail Color, parameterName=Inner_Color_Coverage, sliderMax=1}), X=@n24) {constY=0.52}
  n27 = Multiply(A=@n25, B=LinearInterpolate(B=ScalarParameter() {defaultValue=0.1, parameterName=InnerAlpha}, Alpha=@n26) {constA=1, constB=0.4})
  n28 = LinearInterpolate(A=LinearInterpolate(A=@n27, B=@n26))
  LinearInterpolate(A=@n27, B=@n28, Alpha=ScalarParameter() {group=Trail Alpha Values, parameterName=Use_Stepped_Alpha, sliderMax=1, sortPriority=40}) {constAlpha=0}
  n29 = ParticleColor()
  n30 = LinearInterpolate(B=VectorParameter() {group=Trail Color, parameterName=Outer_Color, defaultValue=<rgba 1 0.03 0.7385 1>}.RGB, Alpha=@n25)
  n31 = VectorParameter() {group=Trail Color, parameterName=Inner_Color, defaultValue=<rgba 1.9243 0.26 2 1>}
  MaterialFunctionCall(Emissive=Multiply(A=@n29.RGB, B=LinearInterpolate(A=@n30, B=@n31.RGB, Alpha=@n26))) {materialFunction=&MF_FX_EmissiveAdjustments}
  n32 = Saturate(None=Multiply(A=ScalarParameter() {defaultValue=1.5, group=Trail Alpha Values, parameterName=Trail_Alpha_Intensity}, B=@n27) {constB=1.2})
  n33 = DepthFade(Opacity=@n32, FadeDistance=ScalarParameter() {group=Alpha, parameterName=FadeDistance, sortPriority=2})
  n34 = DynamicParameter() {paramNames [X,Y,Z,Param4]}
  n35 = SphereMask(
   A=WorldPosition().XYZ
   B=MaterialFunctionCall(R=@n34.X, G=@n34.X, B=@n34.X) {materialFunction=&AppendMany}.RGB
   Radius=@n34.Param4
  ) {attenuationRadius=10, hardnessPercent=90}
  Multiply(A=@n29.A, B=Multiply(A=ShadingPathSwitch(Default=@n33, Mobile=@n32) {desc=Migrated from FeatureLevelSwitch}, B=OneMinus(None=@n35)))
  OneMinus(None=@n28)
 ]
}
